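# Sound cue for the sit foley of the WhineDummeez quadruped cosmetic companion.
# A random node picks one of three equally weighted sit recordings, played at reduced volume and slightly lowered pitch.
# It uses the pets creature sound class and companion foley attenuation, and shares the quadruped footstep concurrency set at zero priority.
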SoundCue /CosmeticCompanions/Assets/Quadruped/WhineDummeez/Audio/Foley/SC_MImosa_Quadraped_WhineDummeez_Sit
root=Random_0
nodes [
 {id=Random_0, children [WavePlayer_3,WavePlayer_4,WavePlayer_5], properties {weights [1,1,1]}}
 {
  id=WavePlayer_3
  properties {soundWaveAssetPtr=/CosmeticCompanions/Assets/Quadruped/WhineDummeez/Audio/Foley/Wavs/Mimosa_Quadraped_WhineDummeez_Sit_01}
 }
 {
  id=WavePlayer_4
  properties {soundWaveAssetPtr=/CosmeticCompanions/Assets/Quadruped/WhineDummeez/Audio/Foley/Wavs/Mimosa_Quadraped_WhineDummeez_Sit_02}
 }
 {
  id=WavePlayer_5
  properties {soundWaveAssetPtr=/CosmeticCompanions/Assets/Quadruped/WhineDummeez/Audio/Foley/Wavs/Mimosa_Quadraped_WhineDummeez_Sit_03}
 }
]
properties {
 attenuationSettings=/Game/Sounds/Attenuation/AS_Companion_Foley
 maxDistance=1200
 pitchMultiplier=0.9
 priority=0
 soundClassObject=&Fort_Creatures_Pets
 volumeMultiplier=0.25
 concurrencySet [/CosmeticCompanions/Assets/Audio_Shared/Foley/Quadraped/CONC_Mimosa_Quadraped_Footstep]
}
