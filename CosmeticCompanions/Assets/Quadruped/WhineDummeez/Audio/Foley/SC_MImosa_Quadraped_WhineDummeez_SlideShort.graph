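# Sound cue for the short slide foley of the WhineDummeez quadruped cosmetic companion.
# A quality level node sends the first two levels to a random pick from ten equally weighted wave variants, and the last two to a reduced pool of the first four.
# Playback uses the companion foley attenuation, the creature pets sound class and the shared quadruped footstep concurrency, at reduced volume.
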SoundCue /CosmeticCompanions/Assets/Quadruped/WhineDummeez/Audio/Foley/SC_MImosa_Quadraped_WhineDummeez_SlideShort
root=QualityLevel_0
nodes [
 {id=QualityLevel_0, children [Random_2,Random_2,Random_4,Random_4]}
 {
  id=Random_2
  children [WavePlayer_0,WavePlayer_1,WavePlayer_2,WavePlayer_3,WavePlayer_4,WavePlayer_5,WavePlayer_6,WavePlayer_7,WavePlayer_8,WavePlayer_9]
  properties {weights [1,1,1,1,1,1,1,1,1,1]}
 }
 {id=Random_4, children [WavePlayer_10, WavePlayer_11, WavePlayer_12, WavePlayer_13], properties {weights [1,1,1,1]}}
 {
  id=WavePlayer_0
  properties {soundWaveAssetPtr=/CosmeticCompanions/Assets/Quadruped/WhineDummeez/Audio/Foley/Wavs/Mimosa_Quadraped_WhineDummeez_SlideShort_10}
 }
 {
  id=WavePlayer_1
  properties {soundWaveAssetPtr=/CosmeticCompanions/Assets/Quadruped/WhineDummeez/Audio/Foley/Wavs/Mimosa_Quadraped_WhineDummeez_SlideShort_09}
 }
 {
  id=WavePlayer_10
  properties {soundWaveAssetPtr=/CosmeticCompanions/Assets/Quadruped/WhineDummeez/Audio/Foley/Wavs/Mimosa_Quadraped_WhineDummeez_SlideShort_04}
 }
 {
  id=WavePlayer_11
  properties {soundWaveAssetPtr=/CosmeticCompanions/Assets/Quadruped/WhineDummeez/Audio/Foley/Wavs/Mimosa_Quadraped_WhineDummeez_SlideShort_03}
 }
 {
  id=WavePlayer_12
  properties {soundWaveAssetPtr=/CosmeticCompanions/Assets/Quadruped/WhineDummeez/Audio/Foley/Wavs/Mimosa_Quadraped_WhineDummeez_SlideShort_02}
 }
 {
  id=WavePlayer_13
  properties {soundWaveAssetPtr=/CosmeticCompanions/Assets/Quadruped/WhineDummeez/Audio/Foley/Wavs/Mimosa_Quadraped_WhineDummeez_SlideShort_01}
 }
 {
  id=WavePlayer_2
  properties {soundWaveAssetPtr=/CosmeticCompanions/Assets/Quadruped/WhineDummeez/Audio/Foley/Wavs/Mimosa_Quadraped_WhineDummeez_SlideShort_08}
 }
 {
  id=WavePlayer_3
  properties {soundWaveAssetPtr=/CosmeticCompanions/Assets/Quadruped/WhineDummeez/Audio/Foley/Wavs/Mimosa_Quadraped_WhineDummeez_SlideShort_07}
 }
 {
  id=WavePlayer_4
  properties {soundWaveAssetPtr=/CosmeticCompanions/Assets/Quadruped/WhineDummeez/Audio/Foley/Wavs/Mimosa_Quadraped_WhineDummeez_SlideShort_06}
 }
 {
  id=WavePlayer_5
  properties {soundWaveAssetPtr=/CosmeticCompanions/Assets/Quadruped/WhineDummeez/Audio/Foley/Wavs/Mimosa_Quadraped_WhineDummeez_SlideShort_05}
 }
 {
  id=WavePlayer_6
  properties {soundWaveAssetPtr=&Mimosa_Quadraped_WhineDummeez_SlideShort_04}
 }
 {
  id=WavePlayer_7
  properties {soundWaveAssetPtr=&Mimosa_Quadraped_WhineDummeez_SlideShort_03}
 }
 {
  id=WavePlayer_8
  properties {soundWaveAssetPtr=&Mimosa_Quadraped_WhineDummeez_SlideShort_02}
 }
 {
  id=WavePlayer_9
  properties {soundWaveAssetPtr=&Mimosa_Quadraped_WhineDummeez_SlideShort_01}
 }
]
properties {
 attenuationSettings=/Game/Sounds/Attenuation/AS_Companion_Foley
 maxDistance=1200
 priority=0
 soundClassObject=&Fort_Creatures_Pets
 volumeMultiplier=0.3
 concurrencySet [/CosmeticCompanions/Assets/Audio_Shared/Foley/Quadraped/CONC_Mimosa_Quadraped_Footstep]
}
